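# Niagara trail effect for the Winged Medieval glider, drawing six ribbon anim trails from bone or socket pairs on its skeletal mesh.
# A single emitter uses SpawnGroup-indexed arrays to vary each trail's width, lifetime, offset, motion and material UVs.
# Trail velocity follows User.ForwardDot in gameplay or switches to front-end values with vertical acceleration, while updates inherit owner movement.
# Timer-driven alpha and time-of-day brightness scale the color across two ribbon renderers.
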NiagaraSystem /BRCosmetics/Gadgets/Assets/VinderTech_GliderChute/Glider_WingedMedieval/FX/NS_Glider_WingedMedieval
dependencies {
 usedDataInterfaces [
  &NiagaraDataInterfaceArrayFloat
  &NiagaraDataInterfaceArrayFloat3
  &NiagaraDataInterfaceCurve
  &NiagaraDataInterfaceSkeletalMesh
 ]
 usedDynamicInputs [&SetBoolByFloatComparison]
 usedModules [
  &self:SystemSpawnScript.SystemScriptSource.SystemScriptGraph.NiagaraNodeAssignment_0.SetVariables_#1
  &SystemState
  &FN_Integrator
  &FN_Timer
  &Timeline
 ]
}
emitters {FN_AnimTrail {
  emitterSpawnScript [&FN_RibbonIDOffsetUtility()]
  emitterUpdateScript [
   &EmitterState(Life Cycle Mode=::NewEnumerator1, Loop Behavior=::NewEnumerator1, Loop Duration Mode=::NewEnumerator1, UseLoopDelay=false)
   &FN_SpawnAnimTrail(Number of Anim Trails=6)
  ]
  particleSpawnScript [
   &FN_InitializeAnimTrailBoneIndecies()
   &FN_InitializeAnimTrailBoneIndecies(AnimTrail Selection=1, First Bone or Socket Index=2, Second Bone or Socket Index=3) {module=FN_InitializeAnimTrailBoneIndecies001}
   &FN_InitializeAnimTrailBoneIndecies(AnimTrail Selection=2, First Bone or Socket Index=4, Second Bone or Socket Index=5) {module=FN_InitializeAnimTrailBoneIndecies002}
   &FN_InitializeAnimTrailBoneIndecies(AnimTrail Selection=3, First Bone or Socket Index=6, Second Bone or Socket Index=7) {module=FN_InitializeAnimTrailBoneIndecies003}
   &FN_InitializeAnimTrailBoneIndecies(AnimTrail Selection=4, First Bone or Socket Index=8, Second Bone or Socket Index=9) {module=FN_InitializeAnimTrailBoneIndecies004}
   &FN_InitializeAnimTrailBoneIndecies(AnimTrail Selection=5, First Bone or Socket Index=10, Second Bone or Socket Index=11) {module=FN_InitializeAnimTrailBoneIndecies005}
   &FN_InitializeAnimTrail(
    AnimTrailWidth=&SelectFloatFromArray(Direct Array Index=$Particles.SpawnGroup, Float Selection Array=DataInterface() {propertyValues {FloatData [2,1.8,2,2,1.8,2]}})
    Color=<rgba 0.125 0.0278 0.0803 2.5>
    Lifetime=&SelectFloatFromArray(Direct Array Index=$Particles.SpawnGroup, Float Selection Array=DataInterface() {propertyValues {FloatData [0.5,0.45,0.4,0.5,0.45,0.4]}})
    Position Offset=&SelectVectorFromArray(
     Direct Array Index=$Particles.SpawnGroup
     Vector Selection Array=DataInterface() {propertyValues {FloatData [<-25 0 0>, <-30 0 0>, <-25 5 0>, <-25 0 0>, <-30 0 0>, <-25 -5 0>]}}
    )
    SkeletalMesh=$System.SK_Mesh
   )
   &AddVelocity(Velocity=<0 400 0>) {enabled=false}
   &AddVelocity(Velocity=&MakeVector(
     X=&MakeCustomFloatFromBool(
      Bool=$User.bIsFrontEnd
      True Float=900
      False Float=&FloatFromCurve(
       FloatCurve=DataInterface() {propertyValues {Curve {keys [
           {arriveTangent=1.2487, interpMode=RCIM_Cubic, leaveTangent=1.2487, tangentMode=RCTM_User, time=0, value=0.25}
           {interpMode=RCIM_Cubic, tangentMode=RCTM_User, time=1, value=1}
          ]}}}
       CurveIndex=&RemapRange(Input Value To Remap=$User.ForwardDot, Input Min=-1, Input Max=1, Output Min=0, Output Max=1)
       Scale Curve=&SelectFloatFromArray(Direct Array Index=$Particles.SpawnGroup, Float Selection Array=DataInterface() {propertyValues {FloatData [700,700,1000,700,700,1000]}})
      )
     )
     Y=0
     Z=&MakeCustomFloatFromBool(
      Bool=$User.bIsFrontEnd
      True Float=&FloatFromCurve(
       FloatCurve=DataInterface() {propertyValues {Curve {keys [
           {arriveTangent=1.501, interpMode=RCIM_Cubic, leaveTangent=1.501, tangentMode=RCTM_User, time=0, value=0.1}
           {arriveTangent=5.3916, interpMode=RCIM_Cubic, leaveTangent=5.3916, tangentMode=RCTM_User, time=0.3, value=1}
          ]}}}
       Scale Curve=1500
      )
      False Float=0
     )
    )) {module=AddVelocity001}
  ]
  particleUpdateScript [
   &InheritSourceMovement(
    Applied Position Scale=&MakeVector(
     X=&FloatFromCurve(
      FloatCurve=DataInterface() {propertyValues {Curve {keys [
          {arriveTangent=-0.1422, interpMode=RCIM_Cubic, leaveTangent=-0.1422, tangentMode=RCTM_User, time=0, value=1}
          {arriveTangent=-1.4907, interpMode=RCIM_Cubic, leaveTangent=-1.4907, tangentMode=RCTM_User, time=0.7, value=0.75}
          {arriveTangent=-2.0475, interpMode=RCIM_Cubic, leaveTangent=-2.0475, tangentMode=RCTM_User, time=1, value=0.3}
         ]}}}
      CurveIndex=&ReturnNormalizedExecIndex()
      Scale Curve=&SelectFloatFromArray(Direct Array Index=$Particles.SpawnGroup, Float Selection Array=DataInterface() {propertyValues {FloatData [-0.7,-0.7,0,-0.7,-0.7,0]}})
     )
     Y=&FloatFromCurve(
      FloatCurve=DataInterface() {propertyValues {Curve {keys [
          {arriveTangent=-0.1422, interpMode=RCIM_Cubic, leaveTangent=-0.1422, tangentMode=RCTM_User, time=0, value=1}
          {arriveTangent=-1.4907, interpMode=RCIM_Cubic, leaveTangent=-1.4907, tangentMode=RCTM_User, time=0.7, value=0.75}
          {arriveTangent=-2.0475, interpMode=RCIM_Cubic, leaveTangent=-2.0475, tangentMode=RCTM_User, time=1, value=0.3}
         ]}}}
      CurveIndex=&ReturnNormalizedExecIndex()
      Scale Curve=&SelectFloatFromArray(
       Direct Array Index=$Particles.SpawnGroup
       Float Selection Array=DataInterface() {propertyValues {FloatData [-0.5,-0.35,-0.25,-0.5,-0.35,-0.25]}}
      )
     )
     Z=&FloatFromCurve(
      FloatCurve=DataInterface() {propertyValues {Curve {keys [
          {arriveTangent=-0.1422, interpMode=RCIM_Cubic, leaveTangent=-0.1422, tangentMode=RCTM_User, time=0, value=1}
          {arriveTangent=-2.0475, interpMode=RCIM_Cubic, leaveTangent=-2.0475, tangentMode=RCTM_User, time=1, value=0.3}
         ]}}}
      Scale Curve=0.15
     )
    )
    ApplyForce=false
    ApplyPosition=true
    ApplyPositionFalloff=false
    Source Velocity:Vector3f=&TransformVector(Vector=$Engine.Owner.Velocity, Source Space=::World, Destination Space=::Local, Apply Scale=true)
   )
   &AccelerationForce(
    Acceleration=&MakeVector(
     X=0
     Y=0
     Z=&MakeCustomFloatFromBool(
      Bool=$User.bIsFrontEnd
      True Float=&FloatFromCurve(FloatCurve=DataInterface() {propertyValues {Curve {keys [<at 0 -1>,<at 1 2.5>]}}}, Scale Curve=2000)
      False Float=0
     )
    )
    Coordinate Space=::Simulation
   )
   &SolveForcesAndVelocity(Clamp Velocity=false, Limit Acceleration=false)
   &ScaleColor(
    Scale Alpha=&FloatFromCurve(
     FloatCurve=DataInterface() {propertyValues {Curve {keys [
         {arriveTangent=4.4472, interpMode=RCIM_Cubic, leaveTangent=4.4472, tangentMode=RCTM_User, time=0, value=0}
         {arriveTangent=1.3333, interpMode=RCIM_Cubic, leaveTangent=1.3333, time=0.028, value=0.3253}
         {interpMode=RCIM_Cubic, leaveTangent=0.1726, tangentMode=RCTM_Break, time=0.3, value=0.4}
         {interpMode=RCIM_Cubic, tangentMode=RCTM_User, time=1, value=0}
        ]}}}
     Scale Curve=&FloatFromCurve(
      FloatCurve=DataInterface() {propertyValues {Curve {keys [
          {arriveTangent=-0.0318, interpMode=RCIM_Cubic, leaveTangent=-0.0318, tangentMode=RCTM_User, time=0, value=0.65}
          {arriveTangent=0.5819, interpMode=RCIM_Cubic, leaveTangent=0.5819, tangentMode=RCTM_User, time=1, value=0.9}
         ]}}}
      CurveIndex=$System.FN_Timer.Value
      Scale Curve=1.5
     )
    )
    Scale RGB=&VectorFromFloat(Value=&Multiply_Float(
      A=&FN_FrontEndFloatModifier(Gameplay=1, Front End=0.9)
      B=&FNTimeOfDayScalar(Night Scale=2.5, Morning Scale=2.25, Day Scale=2.25, Evening Scale=2.25)
     ))
    ScaleA=true
    ScaleRGB=true
   )
   &FN_UpdateAnimTrail(Locked Particle RibbonWidth=1, SkeletalMesh=$System.SK_Mesh) {module=FN_UpdateAnimTrail001}
   /SetVariables(Particles.UVRandomOffset:NiagaraFloat=&SelectFloatFromArray(
     Direct Array Index=&MakeIntFromFloat(Float=&MakeFloatFromInt(Int=$Particles.SpawnGroup))
     Direct Array Mode=::NewEnumerator1
     Float Selection Array=DataInterface() {propertyValues {FloatData [0,0.2,0.45,0.75,0.1,0.3]}}
    )) {isSetParameters=true}
   &DynamicMaterialParameters(
    Index 0 Param 1=1.5
    Index 0 Param 2=$Particles.UVRandomOffset
    Index 0 Param 3=2
    Index 0 Param 4=-5
    Index 1 Param 1=&SelectFloatFromArray(
     Direct Array Index=&MakeIntFromFloat(Float=&MakeFloatFromInt(Int=$Particles.SpawnGroup))
     Direct Array Mode=::NewEnumerator1
     Float Selection Array=DataInterface() {propertyValues {FloatData [3.3,2,2.9,3.15,1.5,2.7]}}
    )
    Index 1 Param 2=1
    Module.Param0WriteEnabled=true
    Module.Param0WriteEnabled1=true
    Module.Param1WriteEnabled=true
    Module.Param1WriteEnabled1=false
    Module.Param2WriteEnabled=true
    Module.Param2WriteEnabled1=false
    Module.Param3WriteEnabled=false
    Module.Param3WriteEnabled1=false
    Write Parameter Index 1=true
   ) {module=DynamicMaterialParameters001}
  ]
  rendererClasses [&NiagaraRibbonRendererProperties]
  renderers [
   {rendererClass=&NiagaraRibbonRendererProperties}
   {rendererClass=&NiagaraRibbonRendererProperties, rendererIndex=1}
  ]
 }}
summary {
 systemName=NS_Glider_WingedMedieval
 emitters [{emitterName=FN_AnimTrail, rendererClasses [&NiagaraRibbonRendererProperties]}]
 userVariables [
  User.bIsFrontEnd {type=NiagaraBool}
  User.bIsFrontEndPreview {type=NiagaraBool}
  User.bisFullyDeployed {type=NiagaraBool}
  User.Delay {type=NiagaraFloat, defaultValue=0, description=""}
  User.ForwardDot {type=NiagaraFloat, defaultValue=0, description=""}
  User.IsFrontEnd {type=NiagaraBool}
  User.RibbonIDOffset {type=NiagaraInt32}
 ]
}
